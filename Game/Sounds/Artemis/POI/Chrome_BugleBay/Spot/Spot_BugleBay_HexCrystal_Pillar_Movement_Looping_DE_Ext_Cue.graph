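# Looping ambient spot SoundCue for hex crystal pillar movement at the Chrome Bugle Bay POI, reusing Herald Fortress waves and attenuation.
# Each loop waits a random delay, then makes a weighted pick between the D and E sets, favoring D, and plays one random, volume-modulated variation.
# A quality switch feeds only the first two levels, leaving the others empty.
# Concurrency override caps simultaneous voices at two.
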SoundCue /Game/Sounds/Artemis/POI/Chrome_BugleBay/Spot/Spot_BugleBay_HexCrystal_Pillar_Movement_Looping_DE_Ext_Cue
root=Looping_0
nodes [
 {id=Delay_0, children [QualityLevel_1], properties {delayMax=12, delayMin=3}}
 {id=Looping_0, children [Delay_0]}
 {id=Modulator_0, children [Random_0], properties {volumeMax=1.2}}
 {id=Modulator_1, children [Random_2], properties {volumeMax=1.2}}
 {id=QualityLevel_1, children [Random_1,Random_1,null,null]}
 {
  id=Random_0
  children [WavePlayer_0, WavePlayer_12, WavePlayer_13, WavePlayer_14, WavePlayer_15, WavePlayer_16]
  properties {weights [1,1,1,1,1,1]}
 }
 {id=Random_1, children [Modulator_0,Modulator_1], properties {weights [1,0.75]}}
 {id=Random_2, children [WavePlayer_1,WavePlayer_2,WavePlayer_3,WavePlayer_4,WavePlayer_5,WavePlayer_6], properties {weights [1,1,1,1,1,1]}}
 {id=WavePlayer_0, properties {soundWaveAssetPtr=/Game/Sounds/Artemis/POI/HerraldsFortress/Spots/Spot_Herald_Fortress_D_01}}
 {id=WavePlayer_1, properties {soundWaveAssetPtr=/Game/Sounds/Artemis/POI/HerraldsFortress/Spots/Spot_Herald_Fortress_E_06}}
 {id=WavePlayer_12, properties {soundWaveAssetPtr=/Game/Sounds/Artemis/POI/HerraldsFortress/Spots/Spot_Herald_Fortress_D_02}}
 {id=WavePlayer_13, properties {soundWaveAssetPtr=/Game/Sounds/Artemis/POI/HerraldsFortress/Spots/Spot_Herald_Fortress_D_03}}
 {id=WavePlayer_14, properties {soundWaveAssetPtr=/Game/Sounds/Artemis/POI/HerraldsFortress/Spots/Spot_Herald_Fortress_D_04}}
 {id=WavePlayer_15, properties {soundWaveAssetPtr=/Game/Sounds/Artemis/POI/HerraldsFortress/Spots/Spot_Herald_Fortress_D_05}}
 {id=WavePlayer_16, properties {soundWaveAssetPtr=/Game/Sounds/Artemis/POI/HerraldsFortress/Spots/Spot_Herald_Fortress_D_06}}
 {id=WavePlayer_2, properties {soundWaveAssetPtr=/Game/Sounds/Artemis/POI/HerraldsFortress/Spots/Spot_Herald_Fortress_E_01}}
 {id=WavePlayer_3, properties {soundWaveAssetPtr=/Game/Sounds/Artemis/POI/HerraldsFortress/Spots/Spot_Herald_Fortress_E_02}}
 {id=WavePlayer_4, properties {soundWaveAssetPtr=/Game/Sounds/Artemis/POI/HerraldsFortress/Spots/Spot_Herald_Fortress_E_03}}
 {id=WavePlayer_5, properties {soundWaveAssetPtr=/Game/Sounds/Artemis/POI/HerraldsFortress/Spots/Spot_Herald_Fortress_E_04}}
 {id=WavePlayer_6, properties {soundWaveAssetPtr=/Game/Sounds/Artemis/POI/HerraldsFortress/Spots/Spot_Herald_Fortress_E_05}}
]
properties {
 attenuationSettings=/Game/Sounds/Artemis/POI/HerraldsFortress/Spots/Spot_Herald_Fortress_Chrome_Movement_Smaller_Att
 bOverrideConcurrency=true
 maxDistance=5000
 soundClassObject=&Fort_Ambient_Outdoors_Athena
 volumeMultiplier=0.752
 attenuationOverrides {falloffDistance=1600}
 concurrencyOverrides {maxCount=2, voiceStealReleaseTime=3}
}
