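# Spot ambience Sound Cue that plays intermittent exterior wooden ship creaks for the Pirate Cove in the Bugle Bay POI.
# A looping chain applies pitch and volume modulation and a random delay, then picks one of seven equally weighted creak waves.
# The Quality Level node feeds only its first two slots, leaving the other two silent.
# It uses outdoor ambient sound class, creak attenuation and a low voice-count concurrency override.
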SoundCue /Game/Sounds/Artemis/POI/Chrome_BugleBay/Spot/Spot_PirateCove_Wooden_Creaks_Ext_Cue
root=Looping_0
nodes [
 {id=Delay_0, children [QualityLevel_0], properties {delayMax=8, delayMin=3}}
 {id=Looping_0, children [Modulator_0]}
 {id=Modulator_0, children [Delay_0], properties {pitchMin=0.85, volumeMax=1.2}}
 {id=QualityLevel_0, children [Random_2,Random_2,null,null]}
 {
  id=Random_2
  children [WavePlayer_13, WavePlayer_14, WavePlayer_15, WavePlayer_16, WavePlayer_17, WavePlayer_18, WavePlayer_19]
  properties {weights [1,1,1,1,1,1,1]}
 }
 {id=WavePlayer_13, properties {soundWaveAssetPtr=/Game/Sounds/Artemis/POI/Chrome_BugleBay/Spot_PirateCove_Ship_Wood_Creak_Ext_01}}
 {id=WavePlayer_14, properties {soundWaveAssetPtr=/Game/Sounds/Artemis/POI/Chrome_BugleBay/Spot_PirateCove_Ship_Wood_Creak_Ext_02}}
 {id=WavePlayer_15, properties {soundWaveAssetPtr=/Game/Sounds/Artemis/POI/Chrome_BugleBay/Spot_PirateCove_Ship_Wood_Creak_Ext_03}}
 {id=WavePlayer_16, properties {soundWaveAssetPtr=/Game/Sounds/Artemis/POI/Chrome_BugleBay/Spot_PirateCove_Ship_Wood_Creak_Ext_04}}
 {id=WavePlayer_17, properties {soundWaveAssetPtr=/Game/Sounds/Artemis/POI/Chrome_BugleBay/Spot_PirateCove_Ship_Wood_Creak_Ext_05}}
 {id=WavePlayer_18, properties {soundWaveAssetPtr=/Game/Sounds/Artemis/POI/Chrome_BugleBay/Spot_PirateCove_Ship_Wood_Creak_Ext_06}}
 {id=WavePlayer_19, properties {soundWaveAssetPtr=/Game/Sounds/Artemis/POI/Chrome_BugleBay/Spot_PirateCove_Ship_Wood_Creak_Ext_07}}
]
properties {
 attenuationSettings=&/Fort_Oneshot_Wood_Creak_Ext_Att
 bOverrideConcurrency=true
 maxDistance=2250
 soundClassObject=&Fort_Ambient_Outdoors_Athena
 volumeMultiplier=0.501
 concurrencyOverrides {maxCount=2, voiceStealReleaseTime=2}
 preEffectBusSends [{}]
}
