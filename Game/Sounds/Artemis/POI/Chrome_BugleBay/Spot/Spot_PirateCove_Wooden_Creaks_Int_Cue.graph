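# Sound cue for looping spot ambience of interior ship-wood creaks at the Pirate Cove in the Chrome_BugleBay POI.
# A loop repeats a randomized delay of several seconds, then plays one of eight equally weighted creak waves with pitch and volume variation.
# A quality switch feeds creaks only to its first two levels, leaving the others silent.
# It uses the indoor ambient sound class, shared creak attenuation and a tight concurrency override.
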SoundCue /Game/Sounds/Artemis/POI/Chrome_BugleBay/Spot/Spot_PirateCove_Wooden_Creaks_Int_Cue
root=Looping_0
nodes [
 {id=Delay_0, children [QualityLevel_1], properties {delayMax=12, delayMin=3}}
 {id=Looping_0, children [Modulator_0]}
 {id=Mixer_0, children [Random_0], properties {inputVolume [1]}}
 {id=Modulator_0, children [Delay_0], properties {pitchMin=0.85, volumeMax=1.2, volumeMin=0.9}}
 {id=QualityLevel_1, children [Mixer_0,Mixer_0,null,null]}
 {
  id=Random_0
  children [WavePlayer_3, WavePlayer_4, WavePlayer_5, WavePlayer_8, WavePlayer_9, WavePlayer_10, WavePlayer_11, WavePlayer_12]
  properties {weights [1,1,1,1,1,1,1,1]}
 }
 {id=WavePlayer_10, properties {soundWaveAssetPtr=/Game/Sounds/Artemis/POI/Chrome_BugleBay/Spot_PirateCove_Ship_Wood_Creak_Int_05}}
 {id=WavePlayer_11, properties {soundWaveAssetPtr=/Game/Sounds/Artemis/POI/Chrome_BugleBay/Spot_PirateCove_Ship_Wood_Creak_Int_06}}
 {id=WavePlayer_12, properties {soundWaveAssetPtr=/Game/Sounds/Artemis/POI/Chrome_BugleBay/Spot_PirateCove_Ship_Wood_Creak_Int_07}}
 {id=WavePlayer_3, properties {soundWaveAssetPtr=/Game/Sounds/Artemis/POI/Chrome_BugleBay/Spot_PirateCove_Ship_Wood_Creak_Int_08}}
 {id=WavePlayer_4, properties {soundWaveAssetPtr=/Game/Sounds/Artemis/POI/Chrome_BugleBay/Spot_PirateCove_Ship_Wood_Creak_Int_01}}
 {id=WavePlayer_5, properties {soundWaveAssetPtr=/Game/Sounds/Artemis/POI/Chrome_BugleBay/Spot_PirateCove_Ship_Wood_Creak_Int_02}}
 {id=WavePlayer_8, properties {soundWaveAssetPtr=/Game/Sounds/Artemis/POI/Chrome_BugleBay/Spot_PirateCove_Ship_Wood_Creak_Int_03}}
 {id=WavePlayer_9, properties {soundWaveAssetPtr=/Game/Sounds/Artemis/POI/Chrome_BugleBay/Spot_PirateCove_Ship_Wood_Creak_Int_04}}
]
properties {
 attenuationSettings=&/Fort_Oneshot_Wood_Creak_Int_Att
 bEnableBusSends=false
 bOverrideConcurrency=true
 maxDistance=850
 soundClassObject=&Fort_Ambient_Indoors_Athena
 volumeMultiplier=0.501
 attenuationOverrides {
  bEnableListenerFocus=true
  falloffDistance=6000
  nonFocusVolumeAttenuation=0.75
  occlusionLowPassFilterFrequency=8000
  occlusionTraceChannel=ECC_GameTraceChannel11
  occlusionVolumeAttenuation=0
  reverbDistanceMax=1024
  reverbDistanceMin=0
  reverbWetLevelMax=0
  reverbWetLevelMin=1
  attenuationShapeExtents {x=0}
 }
 concurrencyOverrides {maxCount=2, voiceStealReleaseTime=2}
 soundSubmixSends [{minSendLevel=0, sendLevel=0, soundSubmix=null}]
}
